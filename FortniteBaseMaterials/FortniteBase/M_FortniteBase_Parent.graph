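# Parent Default Lit material for Fortnite surface instances, routing base color, packed AO/specular/roughness/metallic, normal, emissive and opacity mask through a Creative gameplay function.
# Each texture is sampled with tiled UVs and world-aligned projection, blended by a scalar parameter.
# An optional world-aligned detail layer overlays color and blends angle-corrected normals using a color-derived mask.
# Scalar lerps and channel-mask parameters replace static switches and masks, left as "_OLD" nodes.
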Material /FortniteBaseMaterials/FortniteBase/M_FortniteBase_Parent
graph {
 properties {
  bUseMaterialAttributes=true
  bUsedWithInstancedStaticMeshes=true
  bUsedWithMeshDeformer=true
  bUsedWithNanite=true
  bUsedWithNiagaraMeshParticles=true
  bUsedWithSkeletalMesh=true
  bUsedWithSplineMeshes=true
  pixelDepthOffsetMode=PDOM_Legacy
  usedShadingModels=Default Lit
 }
 nodes [
  n1 = TextureObjectParameter() {
   desc="BaseColor texture using Default compression.\r\nLeave the default setting if not used."
   group=07_DetailTexture
   parameterName=DetailTexture_Diffuse
   sortPriority=20
   texture=/EpicBaseTextures/Blanks/T_Default_BaseColor_White
  }
  n2 = ScalarParameter() {
   defaultValue=400
   desc=Size of the DetailTextures in world units.
   group=07_DetailTexture
   parameterName=DetailTexture_Scale
   sortPriority=40
  }
  n3 = ScalarParameter() {
   defaultValue=1
   desc="Sharpens the amount of falloff of the XYZ masks.\r\nA positive number or zero is expected.\r\nNegative numbers may produce unexpected results."
   group=07_DetailTexture
   parameterName=DetailTexture_ProjectionTransitionSharpness
   sortPriority=50
  }
  n4 = StaticBool()
  n5 = TextureObjectParameter() {
   desc="Tangent-space normal map texture using Normalmap compression.\r\nLeave the default setting if not used."
   group=07_DetailTexture
   parameterName=DetailTexture_Normal
   samplerType=SAMPLERTYPE_Normal
   sortPriority=30
   texture=/EpicBaseTextures/Blanks/T_Default_NormalMap
  }
  n6 = MaterialFunctionCall!(
   Diffuse Texture Object=@n1
   Spec Texture Object
   TextureSize=@n2
   --------------------------
   Use seperate settings for Z projection
   Use seperate texture for Z projection
   Texture Size Z projection
   Z Plane Diffuse Texture Object
   Z Plane Spec Texture Object
   "--------------------------      "
   ProjectionTransitionContrast=@n3
   Use Texture Alpha Heightmap to Refine the Transistion=@n4
   HeightLerpTransitionContrast
   PickTheDiffuseHeightChannel
   TheDiffuseHeightChannel
   "--------------------------       "
   NormalTextureObject=@n5
   World Space Normals=@n4
   "--------------------------  "
   Export Float 4 Diffuse=@n4
   Export Float 4 Spec
   "--------------------------     "
   WorldPosition
   World Space Normal
  ) {materialFunction=/Engine/Functions/Engine_MaterialFunctions02/Texturing/WorldAlignedTextures_Complex}
  LinearInterpolate(
   A=ScalarParameter() {defaultValue=-0.2, group=07_DetailTexture, parameterName=DetailTexture_BaseColor_Minimum, sortPriority=999}
   B=ScalarParameter() {defaultValue=5, group=07_DetailTexture, parameterName=DetailTexture_BaseColor_Maximum, sortPriority=999}
   Alpha=ComponentMask(None=@n6.XYZ Diffuse Texture) {r=true}
  )
  MaterialFunctionCall!(
   TextureObject=@n5
   NormalWS=VertexInterpolator(VS=PreSkinnedNormal()).PS
   ProjectionTransitionSharpness=@n3
   Size=@n2
   NormalFlatness
   Position=VertexInterpolator(VS=PreSkinnedPosition()).PS
  ) {materialFunction=/Game/Global/MaterialFunctions/MF_WorldAlignedNormal_Whiteout}
  n7 = SetMaterialAttributes!(
   MaterialAttributes
   Base Color=NamedRerouteUsage()
   Specular=NamedRerouteUsage()
   Roughness=NamedRerouteUsage()
   Metallic=NamedRerouteUsage()
   Normal=NamedRerouteUsage()
   Emissive Color=NamedRerouteUsage()
   Opacity Mask=NamedRerouteUsage()
  ) {attributeSetTypes [
    #1
    #2
    #3
    #4
    #5
    #6
    #7
   ]}
  MaterialFunctionCall!(In=@n7, IsForLandscapeMaterial) {materialFunction=/Game/Creative/Materials/MaterialFunctions/MF_CreativeFortniteGameplay}
  n8 = ScalarParameter() {
   desc="Seamlessly tiles all textures in XYZ worldspace.\r\nDefault is zero.\r\nNote: This will look incorrect on moving objects."
   group=06_TextureCoordinates
   parameterName=UseWorldAlignedTextures
   sortPriority=30
  }
  NamedRerouteDeclaration(None=Saturate(None=@n8)) {name=WAT_Use, nodeColor=<rgba 0 0 0 1>}
  n9 = ScalarParameter() {
   desc="Use a Basecolor && Normal map texture to produce fine, close-up detail.\r\nDefault is zero.\r\nNote: This will look incorrect on moving objects."
   group=07_DetailTexture
   parameterName=UseDetailTexture
   sortPriority=10
  }
  NamedRerouteDeclaration(None=Saturate(None=@n9)) {name=DetailTexture_Use, nodeColor=<rgba 0 0 0 1>}
  n10 = ScalarParameter() {
   desc="Minimum roughness value.\r\nSet minimum && maximum to the same amount to get a singular roughness value."
   group=02_SpecularRoughnessMetallic
   parameterName=Roughness_Minimum
   sortPriority=21
  }
  n11 = ScalarParameter() {
   defaultValue=1
   desc="Maximum roughness value.\r\nSet minimum && maximum to the same amount to get a singular roughness value."
   group=02_SpecularRoughnessMetallic
   parameterName=Roughness_Maximum
   sortPriority=22
  }
  n12 = TextureObjectParameter() {
   desc="Channel-packed texture using Masks compression.\r\nThe red channel can be AmbientOcclusion or Specular, Roughness uses Green channel, Metallic uses Blue channel.\r\nMany of the older Fortnite textures use green channel for metallic and blue channel for roughness.\r\nLeave the default setting if not used."
   group=02_SpecularRoughnessMetallic
   parameterName="AmbientOcclusion_OR_Specular,Roughness,Metallic_Texture"
   samplerType=SAMPLERTYPE_Masks
   sortPriority=-50
   texture=/EpicBaseTextures/Blanks/T_Default_Masks_SpecRoughMetal
  }
  n13 = TextureSample(UVs=NamedRerouteUsage(), Tex=@n12) {samplerSource=SSM_Wrap_WorldGroupSettings, samplerType=SAMPLERTYPE_Masks}
  n14 = MaterialFunctionCall(TextureObject=@n12, TextureSize=NamedRerouteUsage(), ProjectionTransitionContrast=NamedRerouteUsage()) {materialFunction=&WorldAlignedTexture}
  n15 = LinearInterpolate(A=@n13.RGB, B=@n14.XYZ Texture, Alpha=NamedRerouteUsage())
  n16 = ChannelMaskParameter(None=@n15) {
   desc="Define the texture channel used for the Roughness mask.\r\nDefault is green.\r\nMany of the older Fortnite textures use green channel for metallic and blue channel for roughness."
   group=02_SpecularRoughnessMetallic
   maskChannel=Green
   parameterName=RoughnessChannel
   sortPriority=20
   defaultValue=<rgba 0 1 0 0>
  }
  NamedRerouteDeclaration(None=Saturate(None=LinearInterpolate(A=@n10, B=@n11, Alpha=@n16.R))) {name=Roughness, nodeColor=<rgba 0 0 0 1>}
  n17 = ScalarParameter() {
   desc="Minimum metallic value.\r\nSet minimum && maximum to the same amount to get a singular metallic value."
   group=02_SpecularRoughnessMetallic
   parameterName=Metallic_Minimum
   sortPriority=31
  }
  n18 = ScalarParameter() {
   defaultValue=1
   desc="Maximum metallic value.\r\nSet minimum && maximum to the same amount to get a singular metallic value."
   group=02_SpecularRoughnessMetallic
   parameterName=Metallic_Maximum
  }
  n19 = ChannelMaskParameter(None=@n15) {
   desc="Define the texture channel used for the Metallic mask.\r\nDefault is blue.\r\nMany of the older Fortnite textures use green channel for metallic and blue channel for roughness."
   group=02_SpecularRoughnessMetallic
   maskChannel=Blue
   parameterName=MetallicChannel
   sortPriority=30
   defaultValue=<rgba 0 0 1 0>
  }
  NamedRerouteDeclaration(None=Saturate(None=LinearInterpolate(A=@n17, B=@n18, Alpha=@n19.R))) {name=Metallic, nodeColor=<rgba 0 0 0 1>}
  n20 = ScalarParameter() {defaultValue=1, desc=Increase or decrease the total Emissive amount, group=04_Emissive, parameterName=Emissive_Multiply, sortPriority=20}
  n21 = VectorParameter() {
   desc="Multiply the Emissive_Texture by a color.\r\nAlpha value is not used."
   group=04_Emissive
   parameterName=Emissive_ColorTint
   sortPriority=10
   channelNames {a=Not Used, b=Blue Value, g=Green Value, r=Red Value}
   defaultValue=<rgba 1 1 1 1>
  }
  n22 = TextureObjectParameter() {
   desc="Emissive texture using Default compression.\r\nLeave the default setting if not used."
   group=04_Emissive
   parameterName=Emissive_Texture
   sortPriority=0
   texture=/EpicBaseTextures/Blanks/T_Default_EmissiveColor_Black
  }
  n23 = TextureSample(UVs=NamedRerouteUsage(), Tex=@n22) {samplerSource=SSM_Wrap_WorldGroupSettings}
  n24 = MaterialFunctionCall(TextureObject=@n22, TextureSize=NamedRerouteUsage(), ProjectionTransitionContrast=NamedRerouteUsage()) {materialFunction=&WorldAlignedTexture}
  n25 = LinearInterpolate(A=@n23.RGB, B=@n24.XYZ Texture, Alpha=NamedRerouteUsage())
  NamedRerouteDeclaration(None=Multiply(A=@n20, B=Multiply(A=@n21, B=@n25))) {name=Emissive, nodeColor=<rgba 0 0 0 1>}
  n26 = ScalarParameter() {
   defaultValue=1
   desc="Controls the intensity of the normal map texture contribution.\r\nZero equals no normal map texture contribution.\r\nA negative number will invert the red && green channel && may produce unexpected results."
   group=03_Normals
   parameterName=NormalMap_Intensity
   sortPriority=10
  }
  n27 = TextureObjectParameter() {
   desc="Tangent-space normal map texture using Normalmap compression.\r\nLeave the default setting if not used."
   group=03_Normals
   parameterName=NormalMap_Texture
   samplerType=SAMPLERTYPE_Normal
   sortPriority=0
   texture=&T_Default_NormalMap
  }
  n28 = TextureSample(UVs=NamedRerouteUsage(), Tex=@n27) {samplerSource=SSM_Wrap_WorldGroupSettings, samplerType=SAMPLERTYPE_Normal}
  n29 = MaterialFunctionCall(TextureObject=@n27, TextureSize=NamedRerouteUsage(), WorldSpace=StaticBool(), ProjectionTransitionContrast=NamedRerouteUsage()) {materialFunction=&WorldAlignedNormal}
  n30 = LinearInterpolate(A=@n28.RGB, B=@n29.XYZ Texture, Alpha=NamedRerouteUsage())
  n31 = AppendVector(A=Multiply(A=@n26, B=ComponentMask(None=@n30) {g=true, r=true}), B=ComponentMask(None=@n30) {b=true})
  n32 = LinearInterpolate(
   A=@n31
   B=MaterialFunctionCall(BaseNormal=@n31, AdditionalNormal=NamedRerouteUsage()) {materialFunction=&BlendAngleCorrectedNormals}.Result
   Alpha=NamedRerouteUsage()
  )
  n33 = NamedRerouteUsage()
  n34 = TwoSidedSign()
  n35 = Constant() {r=1}
  NamedRerouteDeclaration(None=Multiply(
    A=LinearInterpolate(A=@n31, B=@n32, Alpha=@n33)
    B=MaterialFunctionCall(X=@n34, Y=@n34, Z=@n35) {materialFunction=&MakeFloat3}.Result
   )) {name=Normal, nodeColor=<rgba 0 0 0 1>}
  NamedRerouteDeclaration(None=Multiply(A=NamedRerouteUsage(), B=NamedRerouteUsage())) {name=BaseColor, nodeColor=<rgba 0 0 0 1>}
  n36 = ScalarParameter() {desc=Minimum opacity mask value., group=05_OpacityMask, parameterName=OpacityMask_Minimum, sortPriority=11}
  n37 = ScalarParameter() {defaultValue=1, desc=Maximum opacity mask value., group=05_OpacityMask, parameterName=OpacityMask_Maximum, sortPriority=12}
  n38 = TextureObjectParameter() {
   desc="Opacity Mask texture using Masks compression.\r\nLeave the default setting if not used.\r\nTo enable opacity, the material instance must have (in the Material Property Overrides section) the Blend Mode set to Masked, and if needed, Two Sided set to true."
   group=05_OpacityMask
   parameterName=OpacityMask_Texture
   samplerType=SAMPLERTYPE_Masks
   sortPriority=0
   texture=/EpicBaseTextures/Blanks/T_Default_Masks_AllOne
  }
  n39 = TextureSample(UVs=NamedRerouteUsage(), Tex=@n38) {samplerSource=SSM_Wrap_WorldGroupSettings, samplerType=SAMPLERTYPE_Masks}
  n40 = MaterialFunctionCall(
   TextureObject=@n38
   TextureSize=NamedRerouteUsage()
   Export Float 4=StaticBool() {value=true}
   ProjectionTransitionContrast=NamedRerouteUsage()
  ) {materialFunction=&WorldAlignedTexture}
  n41 = LinearInterpolate(A=@n39.RGBA, B=@n40.XYZ Texture, Alpha=NamedRerouteUsage())
  n42 = ChannelMaskParameter(None=@n41) {
   desc="Define which channel of the texture to use as the oapcity mask.\r\nDefault is alpha."
   group=05_OpacityMask
   maskChannel=Alpha
   parameterName=OpacityMask_TextureChannel
   sortPriority=10
   defaultValue=<rgba 0 0 0 1>
  }
  NamedRerouteDeclaration(None=Saturate(None=LinearInterpolate(A=@n36, B=@n37, Alpha=@n42.R))) {name=OpacityMask, nodeColor=<rgba 0 0 0 1>}
  n43 = VectorParameter() {
   desc="Size of texture in XYZ worldspace units.\r\nAlpha value is not used."
   group=06_TextureCoordinates
   parameterName=WorldAlignedTextures_Size
   sortPriority=40
   channelNames {a=Not Used, b=Z Size, g=Y Size, r=X Size}
   defaultValue=<rgba 128 128 128 1>
  }
  NamedRerouteDeclaration(None=ComponentMask(None=@n43) {b=true, g=true, r=true}) {name=WAT_Size, nodeColor=<rgba 0 0 0 1>}
  n44 = ScalarParameter() {
   defaultValue=2
   desc="Sharpens the amount of falloff of the XYZ masks.\r\nA positive number or zero is expected.\r\nNegative numbers may produce unexpected results."
   group=06_TextureCoordinates
   parameterName=WorldAlignedTextures_ProjectionTransitionSharpness
   sortPriority=50
  }
  NamedRerouteDeclaration(None=@n44) {name=WAT_PTC, nodeColor=<rgba 0 0 0 1>}
  NamedRerouteDeclaration(None=@n6.XYZ Diffuse Texture) {name=DetailTextureDiffuse, nodeColor=<rgba 0 0 0 1>}
  n45 = VectorParameter() {
   desc="Controls the intensity of the Detail normal map texture contribution.\r\nUse values lower than 1 for R && G to reduce intensity.\r\nUse values higher than 1 for R && G to increase intensity.\r\nB should usually be kept at 1.\r\nAlpha value is not used.\r\nA negative numbers may produce unexpected results."
   group=07_DetailTexture
   parameterName=DetailTexture_NormalMapStrength
   sortPriority=100
   defaultValue=<rgba 1 1 1 0>
  }
  NamedRerouteDeclaration(None=Normalize(VectorInput=Multiply(A=@n6.Normal XYZ Texture, B=@n45.RGB))) {name=DetailTextureNormal, nodeColor=<rgba 0 0 0 1>}
  n46 = VectorParameter() {
   desc="Use this RGB value to get the dot product of the modified BaseColor_Texture to create a mask to control DetailTexture visibility.\r\nA value of 1,0,0 will use the Red channel of the BaseColor_Texture.\r\nA value of 0,1,0 will use the Blue channel of the BaseColor_Texture.\r\nA value of 0,0,1 will use the Green channel of the BaseColor_Texture.\r\nAlpha value is not used."
   group=07_DetailTexture
   parameterName=DetailTexture_Mask_DotProduct
   sortPriority=200
   defaultValue=<rgba 1 0 0 1>
  }
  n47 = DotProduct(A=ComponentMask(None=@n46.RGB) {b=true, g=true, r=true}, B=NamedRerouteUsage())
  n48 = OneMinus(None=@n47)
  n49 = ScalarParameter() {desc=Use a value of 1 to invert the mask., group=07_DetailTexture, parameterName=DetailTexture_Mask_InvertDotProduct, sortPriority=210}
  n50 = ScalarParameter() {
   defaultValue=1
   desc="Controls the mask contrast.\r\nA value of 1 equals no change."
   group=07_DetailTexture
   parameterName=DetailTexture_Mask_Power
   sortPriority=220
  }
  n51 = Saturate(None=Power(Base=LinearInterpolate(A=@n47, B=@n48, Alpha=@n49), Exp=@n50))
  n52 = ScalarParameter() {
   defaultValue=1
   desc="Add this value to the mask.\r\nA value of 1 will make the DetailTexture appear everywhere.\r\nA value of 0 equals no change.\r\nA value of -1 will remove the DetailTexture."
   group=07_DetailTexture
   parameterName=DetailTexture_Mask_Add
   sortPriority=230
  }
  NamedRerouteDeclaration(None=Saturate(None=Add(A=@n51, B=@n52))) {name=DiffuseMaskForDetail, nodeColor=<rgba 0 0 0 1>}
  n53 = ScalarParameter() {
   desc="Set the minimum value of the AmbientOcclusion mask used to multiply the BaseColor.\r\nIf using a Specular mask instead, or AmbientOcclusion is not desired, set this value to 1."
   group=02_SpecularRoughnessMetallic
   parameterName=AmbientOcclusion_Minimum
   sortPriority=-1
  }
  n54 = ScalarParameter() {
   defaultValue=1
   desc="Set the maximum value of the AmbientOcclusion mask used to multiply the BaseColor.\r\nIf using a Specular mask instead, or AmbientOcclusion is not desired, set this value to 1."
   group=02_SpecularRoughnessMetallic
   parameterName=AmbientOcclusion_Maximum
   sortPriority=0
  }
  n55 = ChannelMaskParameter(None=@n15) {
   desc="Define the texture channel used for the AmbientOcclusion mask. This is used to darken the specular value.\r\nDefault is red."
   group=02_SpecularRoughnessMetallic
   parameterName=AmbientOcclusionChannel
   sortPriority=-10
  }
  n56 = ScalarParameter() {
   defaultValue=1
   desc="AmbientOcclusion mask power.\r\nThis is done before the minimum and maximum."
   group=02_SpecularRoughnessMetallic
   parameterName=AmbientOcclusion_Power
   sortPriority=-5
  }
  n57 = Saturate(None=LinearInterpolate(A=@n53, B=@n54, Alpha=Power(Base=@n55.R, Exp=@n56)))
  n58 = ScalarParameter() {
   desc="Amount of AmbientOcclusion to multiply against the specular value.\r\nInput is clamped from 0-1.\r\nA value of zero will not use any of the AmbientOcclusion channel mask.."
   group=02_SpecularRoughnessMetallic
   parameterName=AmbientOcclusion_Specular_Amount
   sortPriority=1
  }
  NamedRerouteDeclaration(None=LinearInterpolate(B=@n57, Alpha=Saturate(None=@n58)) {constA=1}) {name=SpecAO, nodeColor=<rgba 0 0 0 1>}
  n59 = VectorParameter() {
   desc="Multiply the BaseColor by a color.\r\nThis is done after BaseColor_Divide.\r\nAlpha value is not used."
   group=01_BaseColor
   parameterName=BaseColor_Tint
   sortPriority=20
   channelNames {a=Not Used, b=Blue Value, g=Green Value, r=Red Value}
   defaultValue=<rgba 1 1 1 1>
  }
  n60 = ComponentMask(None=@n59) {b=true, g=true, r=true}
  n61 = NamedRerouteUsage()
  NamedRerouteDeclaration(None=Saturate(None=Multiply(A=@n60, B=@n61))) {name=PreDetailTexture, nodeColor=<rgba 0 0 0 1>}
  n62 = TextureObjectParameter() {
   desc="BaseColor texture using Default compression.\r\nLeave the default setting if not used."
   group=01_BaseColor
   parameterName=BaseColor_Texture
   sortPriority=0
   texture=&T_Default_BaseColor_White
  }
  n63 = TextureSample(UVs=NamedRerouteUsage(), Tex=@n62) {samplerSource=SSM_Wrap_WorldGroupSettings}
  n64 = MaterialFunctionCall(TextureObject=@n62, TextureSize=NamedRerouteUsage(), ProjectionTransitionContrast=NamedRerouteUsage()) {materialFunction=&WorldAlignedTexture}
  n65 = NamedRerouteUsage()
  NamedRerouteDeclaration(None=ComponentMask(None=LinearInterpolate(A=@n63.RGB, B=@n64.XYZ Texture, Alpha=@n65)) {b=true, g=true, r=true}) {name=RawBaseColor, nodeColor=<rgba 0 0 0 1>}
  n66 = NamedRerouteUsage()
  n67 = NamedRerouteUsage()
  n68 = LinearInterpolate(
   A=@n66
   B=Saturate(None=MaterialFunctionCall(Base=@n66, Blend=@n67) {materialFunction=&Blend_Overlay}.Result)
   Alpha=NamedRerouteUsage()
  )
  NamedRerouteDeclaration(None=LinearInterpolate(A=@n66, B=@n68, Alpha=NamedRerouteUsage())) {name=PostDetailTexture, nodeColor=<rgba 0 0 0 1>}
  n69 = ScalarParameter() {
   desc="Amount of AmbientOcclusion to multiply against the BaseColor.\r\nInput is clamped from 0-1.\r\nA value of zero will not use any of the AmbientOcclusion channel mask.."
   group=01_BaseColor
   parameterName=AmbientOcclusion_BaseColor_Amount
   sortPriority=30
  }
  NamedRerouteDeclaration(None=LinearInterpolate(B=@n57, Alpha=Saturate(None=@n69)) {constA=1}) {name=BasecolorAO, nodeColor=<rgba 0 0 0 1>}
  n70 = ScalarParameter() {
   defaultValue=1
   desc=Tiles the textures in horizonal U space using UV coordinate index:0
   group=06_TextureCoordinates
   parameterName=TextureTile_U
   sortPriority=10
  }
  n71 = ScalarParameter() {
   defaultValue=1
   desc=Tiles the textures in vertical V space using UV coordinate index:0
   group=06_TextureCoordinates
   parameterName=TextureTile_V
   sortPriority=20
  }
  NamedRerouteDeclaration(None=Multiply(A=AppendVector(A=@n70, B=@n71), B=TextureCoordinate())) {name=UV_Coordinates, nodeColor=<rgba 0 0 0 1>}
  n72 = ScalarParameter() {
   desc="Minimum specular value.\r\nSet minimum && maximum to the same amount to get a singular specular value."
   group=02_SpecularRoughnessMetallic
   parameterName=Specular_Minimum
   sortPriority=11
  }
  n73 = ScalarParameter() {
   defaultValue=1
   desc="Maximum specular value.\r\nSet minimum && maximum to the same amount to get a singular specular value."
   group=02_SpecularRoughnessMetallic
   parameterName=Specular_Maximum
   sortPriority=12
  }
  n74 = ChannelMaskParameter(None=@n15) {
   desc="Define the texture channel used for the Specular mask.\r\nDefault is red."
   group=02_SpecularRoughnessMetallic
   parameterName=SpecularChannel
   sortPriority=10
  }
  n75 = LinearInterpolate(A=@n72, B=@n73, Alpha=@n74.R)
  NamedRerouteDeclaration(None=Multiply(A=Saturate(None=@n75), B=NamedRerouteUsage())) {name=Specular, nodeColor=<rgba 0 0 0 1>}
  n76 = VectorParameter() {
   desc="Divide the BaseColor_Texture by a color.\r\nUse this to shift colors. Lower values darken, higher values brighten.\r\nAlpha value is not used."
   group=01_BaseColor
   parameterName=BaseColor_Divide
   sortPriority=10
   channelNames {a=Not Used, b=Blue Value, g=Green Value, r=Red Value}
   defaultValue=<rgba 1 1 1 1>
  }
  Saturate(None=Divide(A=@n61, B=ComponentMask(None=@n76) {b=true, g=true, r=true}))
  StaticComponentMaskParameter(None=@n41) {
   defaultR=true
   desc="Define which channel of the texture to use as the oapcity mask.\r\nDefault is red."
   group=05_OpacityMask
   parameterName=OpacityMask_TextureChannel_OLD
   sortPriority=10
  }
  StaticComponentMaskParameter(None=@n15) {
   defaultR=true
   desc="Define the texture channel used for the Specular mask.\r\nDefault is red."
   group=02_SpecularRoughnessMetallic
   parameterName=SpecularChannel_OLD
   sortPriority=10
  }
  StaticComponentMaskParameter(None=@n15) {
   defaultG=true
   desc="Define the texture channel used for the Roughness mask.\r\nDefault is green.\r\nMany of the older Fortnite textures use green channel for metallic and blue channel for roughness."
   group=02_SpecularRoughnessMetallic
   parameterName=RoughnessChannel_OLD
   sortPriority=20
  }
  StaticComponentMaskParameter(None=@n15) {
   defaultB=true
   desc="Define the texture channel used for the Metallic mask.\r\nDefault is blue.\r\nMany of the older Fortnite textures use green channel for metallic and blue channel for roughness."
   group=02_SpecularRoughnessMetallic
   parameterName=MetallicChannel_OLD
   sortPriority=30
  }
  StaticComponentMaskParameter(None=@n15) {
   defaultR=true
   desc="Define the texture channel used for the AmbientOcclusion mask. This is used to darken the specular value.\r\nDefault is red."
   group=02_SpecularRoughnessMetallic
   parameterName=AmbientOcclusionChannel_OLD
   sortPriority=-10
  }
  StaticSwitchParameter(True=@n32, False=@n31) {
   desc="Use a Basecolor && Normal map texture to produce fine, close-up detail."
   group=07_DetailTexture
   parameterName=UseDetailTexture_OLD
   sortPriority=10
  }
  StaticSwitchParameter(True=@n68, False=@n66) {
   desc="Use a Basecolor && Normal map texture to produce fine, close-up detail."
   group=07_DetailTexture
   parameterName=UseDetailTexture_OLD
   sortPriority=10
  }
  StaticSwitchParameter(True=@n64.XYZ Texture, False=@n63.RGB) {
   desc=Seamlessly tiles all textures in XYZ worldspace
   group=06_TextureCoordinates
   parameterName=UseWorldAlignedTextures_OLD
   sortPriority=30
  }
  StaticSwitchParameter(True=@n14.XYZ Texture, False=@n13.RGB) {
   desc=Seamlessly tiles all textures in XYZ worldspace
   group=06_TextureCoordinates
   parameterName=UseWorldAlignedTextures_OLD
   sortPriority=30
  }
  StaticSwitchParameter(True=@n29.XYZ Texture, False=@n28.RGB) {
   desc=Seamlessly tiles all textures in XYZ worldspace
   group=06_TextureCoordinates
   parameterName=UseWorldAlignedTextures_OLD
   sortPriority=30
  }
  StaticSwitchParameter(True=@n24.XYZ Texture, False=@n23.RGB) {
   desc=Seamlessly tiles all textures in XYZ worldspace
   group=06_TextureCoordinates
   parameterName=UseWorldAlignedTextures_OLD
   sortPriority=30
  }
  StaticSwitchParameter(True=@n40.XYZ Texture, False=@n39.RGBA) {
   desc=Seamlessly tiles all textures in XYZ worldspace
   group=06_TextureCoordinates
   parameterName=UseWorldAlignedTextures_OLD
   sortPriority=30
  }
 ]
}
